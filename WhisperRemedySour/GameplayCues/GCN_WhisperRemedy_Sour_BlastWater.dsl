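; Gameplay cue for the Whisper Remedy Sour water blast, handling the burst event on the target actor.
; It shows a gunshot-type sound indicator at the target's location using the target's team, a Hostile affiliation and the first audio component's sound.
; It then sets increasing burst-delay parameters on the second and third particle components to stagger their timing.
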(fn OnBurst(EitherVFX-preferthisversion)
  (params (MyTarget :type Actor%20Object%20Reference) (Parameters :type Gameplay%20Cue%20Parameters%20Structure%20%28by%20ref%29) (ParticleComponents :type Array%20of%20FXSystem%20Component%20Object%20References) (AudioComponents :type Array%20of%20Audio%20Component%20Object%20References) (BurstCameraShakeInstance :type Camera%20Shake%20Base%20Object%20Reference) (BurstDecalInstance :type Decal%20Actor%20Object%20Reference))
  (Utilities|FlowControl|Sequence
    (:then_0
      (Fortnite|ShowSoundIndicatorLocalforTeam :WorldContextObject MyTarget :TrackedActor MyTarget :Location (Transformation|GetActorLocation MyTarget) :MaxAudibleDistance 8e+03 :IndicatorType "Gunshot" :Team (Team|GetActorTeam MyTarget) :Affiliations (Utilities|Array|MakeArray "Hostile") :Tint "(R=0.000000,G=0.000000,B=0.000000,A=0.000000)" :SoundBase (Variables|/Script/Engine.AudioComponent|GetSound (Utilities|Array|Get(acopy) AudioComponents)) :AttenuationMaxRangeMultiplier 1.0 :AttenuationIntensityMultiplier 1.0))
    (:then_1
      (Effects|Components|ParticleSystem|SetFloatParameter (Utilities|Array|Get(acopy) ParticleComponents 1) "Timing_BurstDelay" 0.2)
      (Effects|Components|ParticleSystem|SetFloatParameter (Utilities|Array|Get(acopy) ParticleComponents 2) "Timing_BurstDelay" 0.4))
    (:then_2
      (return))))
